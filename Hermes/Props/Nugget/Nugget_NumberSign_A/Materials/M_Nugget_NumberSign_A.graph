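# Nanite-ready material for the Nugget number sign prop, showing a selectable number with a tinted emissive glow.
# UV1's floored U blends an orange-tinted texture set A with set B, whose atlas UVs shift in thirds from the per-primitive Number value.
# Shared damage, wind, fan-spin and snow functions follow, then world-space normals with backface correction.
# Low-quality mobile flattens specular and roughness before the Creative move-tool function.
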Material /Hermes/Props/Nugget/Nugget_NumberSign_A/Materials/M_Nugget_NumberSign_A
graph {
 properties {
  bAutomaticallySetUsageInEditor=false
  bTangentSpaceNormal=false
  bUseMaterialAttributes=true
  bUsedWithInstancedStaticMeshes=true
  bUsedWithNanite=true
  ditheredLODTransition=true
  usedShadingModels=Default Lit
 }
 nodes [
  n1 = TextureSample() {samplerType=SAMPLERTYPE_Masks, texture=/Hermes/Props/Nugget/Nugget_NumberSign_A/Textures/T_Nugget_NumberSign_A_A_S}
  n2 = SetMaterialAttributes!(
   MaterialAttributes
   Base Color=Multiply(
    A=TextureSample() {texture=/Hermes/Props/Nugget/Nugget_NumberSign_A/Textures/T_Nugget_NumberSign_A_A_D}.RGB
    B=Constant3Vector() {constant=<rgba 1 0.4287 0 1>}
   )
   Normal=TextureSample() {samplerType=SAMPLERTYPE_Normal, texture=/Hermes/Props/Nugget/Nugget_NumberSign_A/Textures/T_Nugget_NumberSign_A_A_N}.RGB
   Specular=@n1.R
   Metallic=@n1.R
   Roughness=@n1.R
  ) {attributeSetTypes [
    #1
    #2
    #3
    #4
    #5
   ]}
  n3 = TextureSample(UVs=NamedRerouteUsage()) {samplerType=SAMPLERTYPE_Masks, texture=/Hermes/Props/Nugget/Nugget_NumberSign_A/Textures/T_Nugget_NumberSign_A_B_S}
  n4 = NamedRerouteUsage()
  n5 = SetMaterialAttributes!(
   MaterialAttributes
   Base Color=TextureSample(UVs=NamedRerouteUsage()) {texture=/Hermes/Props/Nugget/Nugget_NumberSign_A/Textures/T_Nugget_NumberSign_A_B_D}.RGB
   Normal=TextureSample(UVs=NamedRerouteUsage()) {samplerType=SAMPLERTYPE_Normal, texture=/Hermes/Props/Nugget/Nugget_NumberSign_A/Textures/T_Nugget_NumberSign_A_B_N}.RGB
   Specular=@n3.R
   Metallic=@n3.R
   Roughness=@n3.R
   Anisotropy
   Emissive Color=Multiply(
    A=TextureSample(UVs=@n4) {samplerType=SAMPLERTYPE_Masks, texture=/Hermes/Props/Nugget/Nugget_NumberSign_A/Textures/T_Nugget_NumberSign_A_A_M}.G
    B=VectorParameter() {parameterName=Emissive Tint, defaultValue=<rgba 2 1.5 1 1>}.RGB
   )
  ) {attributeSetTypes [
    #1
    #2
    #3
    #4
    #5
    #6
    #7
   ]}
  n6 = TextureCoordinate() {coordinateIndex=1}
  n7 = BlendMaterialAttributes(A=@n2, B=@n5, Alpha=Floor(None=ComponentMask(None=@n6) {r=true}))
  MaterialExpressionReroute_69 = Reroute(None=@n7)
  n8 = MaterialFunctionCall!(Material Attributes=MaterialFunctionCall!(In=MaterialFunctionCall(Material_In=@MaterialExpressionReroute_69.MaterialAttributes) {materialFunction=&MF_FN_DestructibleDamage}.Material_Out) {materialFunction=/Game/Athena/WinterArt/Materials/MF_WPO_WindNew}.Result) {materialFunction=/Game/Creative/Sets/MilitaryBase/Props/MF_FanSpin}
  n9 = MaterialFunctionCall!(In=MaterialFunctionCall!(
    In=MaterialFunctionCall!(In=@n8.Result, UseSphericalMappingOverride) {materialFunction=/Game/Packages/Fortress_SharedMaterials/Base_Material_Functions/MF_CurieFallbackEffect}.Result
    Custom Mask
    Custom Rotation Axis
   ) {materialFunction=/Game/Athena/WinterArt/Materials/MF_WindWPO_v2}.Result) {materialFunction=/Game/Global/Base4Layers_2022/MaterialFunctions/MF_SnowAndIceArchitecture}
  MaterialExpressionReroute_68 = Reroute(None=@n9.Result)
  Composite!(None=@MaterialExpressionReroute_68.Result) {desc=Transform to World in case we will use RVT after this point, subgraphName=Tangent to World}
  Composite!(MaterialAttributes=@n7) {desc=Game Shared WPO and Fire Effect, subgraphName=WPO And Shared Functions}
  MaterialExpressionReroute_3 = Reroute(None=@MaterialExpressionReroute_68.Result)
  n10 = GetMaterialAttributes(None=@MaterialExpressionReroute_3.Input) {attributeGetTypes [#2]}
  n11 = Transform(None=@n10.Normal)
  n12 = LinearInterpolate(A=Multiply(A=@n11) {constB=-1}, B=@n11, Alpha=Saturate(None=TwoSidedSign())) {desc="Fixes normals\r\non backfaces"}
  n13 = SetMaterialAttributes!(
   MaterialAttributes=@n10.MaterialAttributes
   Normal=StaticSwitchParameter(True=MaterialProxyReplace(Realtime=@n12, MaterialProxy=@n10.Normal), False=@n12) {parameterName=Force TangentSpaceNormalsFor HLOD ProxyReplace}
  ) {attributeSetTypes [#2]}
  MaterialExpressionReroute_2 = Reroute(None=@n13)
  Composite!(InputPin=@MaterialExpressionReroute_2.Output) {subgraphName=Mobile Opt}
  MaterialExpressionReroute_7 = Reroute(None=@MaterialExpressionReroute_2.Output)
  n14 = GetMaterialAttributes(None=@MaterialExpressionReroute_7.InputPin)
  n15 = Constant() {r=0.5}
  n16 = Constant() {r=0.8}
  n17 = ShadingPathSwitch(
   Default=@MaterialExpressionReroute_7.InputPin
   Mobile=QualitySwitch(
    Default=@MaterialExpressionReroute_7.InputPin
    Low=SetMaterialAttributes!(MaterialAttributes=@n14.MaterialAttributes, Specular=@n15, Roughness=@n16) {attributeSetTypes [#3, #5]}
    High=@MaterialExpressionReroute_7.InputPin
    Medium=@MaterialExpressionReroute_7.InputPin
    Epic=@MaterialExpressionReroute_7.InputPin
   )
  )
  MaterialExpressionReroute_6 = Reroute(None=@n17)
  MaterialFunctionCall(MA In=@MaterialExpressionReroute_6.Output) {materialFunction=&MF_CreativeMoveTool}
  n18 = Multiply(A=ScalarParameter() {bUseCustomPrimitiveData=true, defaultValue=1, parameterName=Number}) {constB=0.3333}
  n19 = Multiply(A=Ceil(None=Add(A=@n18) {constB=-0.01})) {constB=0.3333}
  NamedRerouteDeclaration(None=Add(A=TextureCoordinate(), B=AppendVector(A=@n18, B=@n19))) {name=UVNumbers, nodeColor=<rgba 1 0.7734 0 1>}
  PinBase() {reroutePins [InputPin {expression=&self:MaterialExpressionReroute_7}]}
  PinBase!(Output=@n17) {reroutePins [Output {expression=&self:MaterialExpressionReroute_6}]}
  PinBase!() {reroutePins [Input {expression=&self:MaterialExpressionReroute_3}]}
  PinBase!(Output=@n13) {reroutePins [Output {expression=&self:MaterialExpressionReroute_2}]}
  PinBase!() {reroutePins [MaterialAttributes {expression=&self:MaterialExpressionReroute_69}]}
  PinBase!(Result=@n9.Result) {reroutePins [Result {expression=&self:MaterialExpressionReroute_68}]}
 ]
}
